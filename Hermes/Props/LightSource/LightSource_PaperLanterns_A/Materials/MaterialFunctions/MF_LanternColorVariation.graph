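# Material function that gives paper lanterns alternating two-color tints and flickering glow.
# Local X, with World Position Offset subtracted, defines repeating bands that select the color, invertible through Custom Primitive Data.
# The tint affects Base Color and Emissive only where emissive is nonzero.
# Four summed triangle waves, phase-shifted by object pivot and band index, flicker the emissive above a minimum brightness.
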MaterialFunction /Hermes/Props/LightSource/LightSource_PaperLanterns_A/Materials/MaterialFunctions/MF_LanternColorVariation
graph {nodes [
  n1 = GetMaterialAttributes(None=FunctionInput() {inputName=Material, inputType=FunctionInput_MaterialAttributes}) {attributeGetTypes [#1, #2, #3]}
  n2 = VectorParameter() {group=Lanterns, parameterName=Lantern Color 1, defaultValue=<rgba 0.3958 0.3041 0.2742 1>}
  n3 = VectorParameter() {group=Lanterns, parameterName=Lantern Color 2, defaultValue=<rgba 0.6875 0.6875 0.6875 1>}
  n4 = LocalPosition()
  n5 = Transform(None=NamedRerouteUsage()) {transformSourceType=TRANSFORMSOURCE_World, transformType=TRANSFORM_Local}
  n6 = Divide(A=Subtract(A=ComponentMask(None=@n4.XYZ) {r=true}, B=ComponentMask(None=@n5) {r=true})) {constB=210}
  n7 = Add(A=Frac(None=@n6)) {constB=-0.5}
  n8 = Round(None=Multiply(A=Abs(None=@n7)) {constB=2})
  n9 = LinearInterpolate(A=@n8, B=OneMinus(None=@n8), Alpha=ScalarParameter() {bUseCustomPrimitiveData=true, parameterName=Lantern Pattern Invert})
  n10 = LinearInterpolate(B=Multiply(A=LinearInterpolate(A=@n2.RGB, B=@n3.RGB, Alpha=@n9)) {constB=2}, Alpha=Ceil(None=@n1.EmissiveColor)) {constA=1}
  n11 = Multiply(A=@n1.EmissiveColor, B=@n10)
  n12 = MaterialFunctionCall() {materialFunction=&ObjectPivotPoint}
  n13 = MaterialFunctionCall(Float3=Frac(None=Multiply(A=@n12.Object Pivot Location) {constB=0.15})) {materialFunction=&BreakOutFloat3Components}
  n14 = Add(A=Add(A=@n13.R, B=@n13.R), B=@n13.B)
  n15 = Multiply(A=Round(None=Multiply(A=@n6) {constB=2})) {constB=2.658}
  n16 = Multiply(
   A=MaterialFunctionCall(
    R=ScalarParameter() {defaultValue=1.28, group=Lanterns, parameterName=Flicker Speed 1, sortPriority=0}
    G=ScalarParameter() {defaultValue=4.55, group=Lanterns, parameterName=Flicker Speed 2, sortPriority=0}
    B=ScalarParameter() {defaultValue=3.745, group=Lanterns, parameterName=Flicker Speed 3, sortPriority=0}
    A=ScalarParameter() {defaultValue=5.23, group=Lanterns, parameterName=Flicker Speed 4, sortPriority=0}
   ) {materialFunction=&AppendMany}.RGBA
   B=Add(A=Add(A=@n14, B=@n15), B=Time())
  )
  n17 = Multiply(
   A=Add(A=Frac(None=@n16)) {constB=-0.5}
   B=MaterialFunctionCall(
    R=ScalarParameter() {defaultValue=0.7, group=Lanterns, parameterName=Flicker Amplitude 1, sortPriority=0}
    G=ScalarParameter() {defaultValue=0.25, group=Lanterns, parameterName=Flicker Amplitude 2, sortPriority=0}
    B=ScalarParameter() {defaultValue=0.23, group=Lanterns, parameterName=Flicker Amplitude 3, sortPriority=0}
    A=ScalarParameter() {defaultValue=0.15, group=Lanterns, parameterName=Flicker Amplitude 4, sortPriority=0}
   ) {materialFunction=&AppendMany}.RGBA
  ) {constB=4}
  n18 = MaterialFunctionCall(Float4=Multiply(A=Abs(None=@n17)) {constB=2}) {materialFunction=&BreakOutFloat4Components}
  n19 = Add(A=Add(A=@n18.R, B=@n18.R), B=Add(A=@n18.B, B=@n18.B))
  n20 = ScalarParameter() {defaultValue=0.5, group=Lanterns, parameterName=Flicker Variation Offset, sortPriority=0}
  n21 = Saturate(None=Max(A=Add(A=@n19, B=@n20), B=ScalarParameter() {defaultValue=0.8, group=Lanterns, parameterName=Flicker Minimum Brightness, sortPriority=0}))
  FunctionOutput(None=SetMaterialAttributes!(MaterialAttributes=@n1.MaterialAttributes, BaseColor=Multiply(A=@n1.BaseColor, B=@n10), EmissiveColor=Multiply(A=@n11, B=@n21)) {attributeSetTypes [#1, #2]})
  NamedRerouteDeclaration(None=@n1.WorldPositionOffset) {name=WPO, nodeColor=<rgba 1 0.2812 0 1>}
 ]}
